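# Physics asset for the Pickaxe_17 skeletal mesh, a Fortnite two-handed melee pickaxe cosmetic.
# It has a single body on the Melee_TwoHanded_Handle bone, shaped as one rotated box offset along the handle to approximate the pickaxe.
# There are no constraints, so the weapon collides and simulates as one rigid piece.
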PhysicsAsset /BRCosmetics/Weapons/FORT_Melee/Pickaxe_17/Meshes/Pickaxe_17_SK_Pickaxe_17_PhysicsAsset
bodies {Melee_TwoHanded_Handle [{shapeType=Box, extentX=65.3448, extentY=17.4431, extentZ=158.1227, center=<0.3785 1.0851 35.6647>, rotation=<rot -1.8744 -90 -1.3761>}]}
constraints []
